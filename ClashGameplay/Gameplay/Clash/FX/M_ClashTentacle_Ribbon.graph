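# Two-sided masked subsurface material for Clash tentacle effects on Niagara ribbons.
# SplineThicken gives the strip noise-varied base-to-tip width, blending a cylinder normal with panning swirl-noise normals for a tube look.
# Sine/cosine waves along normal and tangent plus panned noise add writhing world position offset, weighted along U and scaled by dynamic parameters.
# A height-lerp cloud texture driven by the Dissolve dynamic parameter, combined with a sphere mask, builds a dissolve mask.
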Material /ClashGameplay/Gameplay/Clash/FX/M_ClashTentacle_Ribbon
graph {
 properties {
  bTangentSpaceNormal=false
  bUsedWithMeshDeformer=true
  bUsedWithNiagaraMeshParticles=true
  bUsedWithNiagaraRibbons=true
  bUsedWithSkeletalMesh=true
  blendMode=BLEND_Masked
  pixelDepthOffsetMode=PDOM_Legacy
  shadingModel=MSM_Subsurface
  twoSided=true
  usedShadingModels=Subsurface
 }
 nodes [
  Constant() {r=1}
  n1 = TextureCoordinate() {uTiling=0.5, vTiling=2}
  n2 = Multiply(A=MaterialFunctionCall!(
    Diffuse 1
    Diffuse 2
    Transistion Phase=DynamicParameter() {defaultValue=<rgba 1 1 1 0>, paramNames [RandomOffset, WPO_Intensity, WPO_Speed, Dissolve]}.Dissolve
    Height Texture 1
    Height Texture 2=TextureSample(UVs=@n1) {samplerType=SAMPLERTYPE_LinearGrayscale, texture=/Game/Athena/Prototype/Terrain/T_CloudCoverage}.R
    Contrast
   ) {materialFunction=/Engine/Functions/Engine_MaterialFunctions02/Texturing/HeightLerpWithTwoHeightMaps}.Alpha)
  n3 = TextureCoordinate()
  n4 = Constant2Vector() {g=0.5, r=0.3}
  n5 = Constant() {r=0.65}
  LinearInterpolate(A=Multiply(A=Saturate(None=@n2), B=SphereMask(A=@n3, B=@n4, Radius=@n5)), Alpha=ScalarParameter() {parameterName=Alpha_0}) {constB=0}
  MaterialFunctionCall(Emissive=VectorParameter() {parameterName=Simple_EmissiveColor, defaultValue=<rgba 0.012 0.0026 0.0017 1>}.RGB) {materialFunction=&MF_FX_EmissiveAdjustments}
  n6 = ScalarParameter() {defaultValue=10, group=WPO, parameterName=WidthBase}
  n7 = DynamicParameter() {paramNames [rand offset, WidthBase, WidthTip, WPO_strength]}
  n8 = VertexColor()
  n9 = LinearInterpolate(Alpha=Multiply(A=TextureSample(UVs=@n8.B) {samplerType=SAMPLERTYPE_Grayscale, texture=&LowResBlurredNoise}.R) {constB=2}) {constA=0.1, constB=0.7}
  n10 = ScalarParameter() {group=WPO, parameterName=WidthTip}
  n11 = TextureCoordinate()
  n12 = Multiply(A=ComponentMask(None=@n11) {r=true}, B=ScalarParameter() {defaultValue=5, group=Normal, parameterName=Normal Tile U})
  n13 = Multiply(A=ComponentMask(None=@n11) {g=true}, B=ScalarParameter() {defaultValue=0.7, group=Normal, parameterName=Normal Tile V})
  n14 = ScalarParameter() {defaultValue=1, group=Normal, parameterName=Normal Pan Speed U}
  n15 = ScalarParameter() {group=Normal, parameterName=Normal Pan Speed V}
  n16 = TextureSample(UVs=Panner(Coordinate=AppendVector(A=@n12, B=@n13), Time=Time() {bOverride_Period=true, period=64}, Speed=AppendVector(A=@n14, B=@n15)) {speedX=-1, speedY=0.1}) {samplerType=SAMPLERTYPE_Normal, texture=/Game/Global/Textures/Default/Noise/T_SwirlNoise_N}
  n17 = VectorParameter() {group=Normal, parameterName=NormalStrength, defaultValue=<rgba 0.4 0.4 1 1>}
  n18 = MaterialFunctionCall(
   WidthBase=Multiply(A=Multiply(A=@n6, B=@n7.WidthBase), B=@n9)
   WidthTip=Multiply(A=Multiply(A=@n7.WidthTip, B=@n10), B=@n9)
   Expand U or V UV Channel=StaticBool()
   AdditionalNormal=MaterialFunctionCall(
    BaseNormal=TextureSample() {samplerType=SAMPLERTYPE_Normal, texture=/Game/Characters/Player/Male/Large/Bodies/M_LRG_Vertigo/FX/T_CylinderNormal}.RGB
    AdditionalNormal=Multiply(A=@n16.RGB, B=@n17.RGB)
   ) {materialFunction=&BlendAngleCorrectedNormals}.Result
  ) {materialFunction=&SplineThicken}
  n19 = TextureCoordinate()
  n20 = OneMinus(None=Power(Base=ComponentMask(None=@n19) {r=true}))
  n21 = Power(Base=@n20)
  n22 = VertexNormalWS()
  n23 = TextureCoordinate()
  n24 = ScalarParameter() {defaultValue=2, group=WPO, parameterName=SinSpeed}
  n25 = Add(A=Multiply(A=ComponentMask(None=@n23) {r=true}), B=Multiply(A=Time(), B=Multiply(A=@n24)))
  n26 = Sine(None=@n25)
  n27 = Cosine(None=@n25)
  n28 = VertexTangentWS()
  n29 = Multiply(
   A=Add(A=Multiply(A=@n22, B=@n26), B=Multiply(A=@n27, B=@n28)) {constB=0}
   B=ScalarParameter() {defaultValue=5, group=WPO, parameterName=SinIntensity}
  )
  n30 = Panner(Coordinate=TextureCoordinate() {vTiling=0.2}) {speedX=0.5}
  n31 = TextureCoordinate()
  n32 = Multiply(
   A=LinearInterpolate(Alpha=TextureSample(UVs=@n30) {samplerType=SAMPLERTYPE_Grayscale, texture=&LowResBlurredNoise}.R) {constA=-1}
   B=Power(Base=ComponentMask(None=@n31) {r=true})
  )
  n33 = Multiply(A=Saturate(None=Multiply(A=@n21) {constB=10}), B=Multiply(A=@n20, B=Add(A=@n29, B=@n32) {constB=0}))
  n34 = DynamicParameter() {parameterIndex=1, paramNames [WPO_strength,Param2,Param3,Param4]}
  Reroute(None=Add(A=Add(A=@n18.WorldPositionOffset) {constB=0}, B=Multiply(A=@n33, B=@n34.WPO_strength)))
  Reroute(None=@n18.Normal)
  ScalarParameter() {defaultValue=1, parameterName=Specular}
  ScalarParameter() {defaultValue=0.4, parameterName=Roughness}
  ScalarParameter() {defaultValue=0.4, parameterName=Metalic}
  ScalarParameter() {defaultValue=1, parameterName=WPO_Intensity}
  VectorParameter() {parameterName=baseColorInner, defaultValue=<rgba 0.224 0.0106 0 1>}
 ]
}
