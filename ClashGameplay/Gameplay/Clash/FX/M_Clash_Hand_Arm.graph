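# Masked lit material for a Clash hand/arm effect that stretches a mesh from its object position toward a TargetPosition, usable on skeletal and Niagara meshes.
# Vertex offset follows UV1's V coordinate, adding curve-atlas-shaped, length-scaled cone inflation, sine/cosine wobble perpendicular to the stretch and texture-driven normal displacement.
# Opacity combines scalar and particle Dissolve values with animated fire noise.
# Tentacle texture drives tinted color and time-of-day emissive, plus a sun-direction highlight from swirl-noise normals.
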Material /ClashGameplay/Gameplay/Clash/FX/M_Clash_Hand_Arm
graph {
 properties {
  bUsedWithMeshDeformer=true
  bUsedWithNiagaraMeshParticles=true
  bUsedWithSkeletalMesh=true
  blendMode=BLEND_Masked
  pixelDepthOffsetMode=PDOM_Legacy
  usedShadingModels=Default Lit
 }
 nodes [
  n1 = Subtract(A=VectorParameter() {parameterName=TargetPosition, defaultValue=<rgba 0 0 0 1>}.RGB, B=ObjectPositionWS())
  n2 = Multiply(A=ComponentMask(None=TextureCoordinate() {coordinateIndex=1}) {g=true}, B=@n1)
  n3 = VertexNormalWS()
  n4 = CurveAtlasRowParameter(CurveTime=ComponentMask(None=TextureCoordinate() {coordinateIndex=1}) {g=true}) {
   atlas=/Game/Athena/LinearColorCurves/TempAthenaCurveAtlas
   curve=/Game/Athena/Prototype/Blueprints/NightNight/FX/FinalImpact/Curve_HT_Tomato_RC
   defaultValue=11
  }
  n5 = Multiply(A=Multiply(
    A=MaterialFunctionCall(Vector 3=@n1) {materialFunction=/Engine/Functions/Engine_MaterialFunctions02/Utility/VectorLength}.V3 Length
    B=ScalarParameter() {defaultValue=0.6, parameterName=ConeSizeScaleBaseOnLength}
   )) {constB=0.002}
  n6 = Multiply(
   A=Multiply(A=Normalize(VectorInput=@n3), B=@n4.G) {constB=0}
   B=Multiply(A=Clamp(None=@n5) {maxDefault=0.5, minDefault=0.2}, B=ScalarParameter() {defaultValue=60, parameterName=ConeIntensity})
  )
  n7 = Multiply(A=ComponentMask(None=TextureCoordinate() {coordinateIndex=1}) {r=true}, B=ScalarParameter() {defaultValue=1, parameterName=U_Scale})
  n8 = Multiply(
   A=SphereMask(A=WorldPosition().XYZ, B=ActorPositionWS(), Radius=ScalarParameter() {defaultValue=3000, parameterName=Radius}) {hardnessPercent=0}
   B=ScalarParameter() {defaultValue=1, parameterName=V_Scale}
  )
  n9 = Multiply(
   A=TextureSampleParameter2D(UVs=AppendVector(A=@n7, B=@n8)) {parameterName=Normal, samplerType=SAMPLERTYPE_Normal, texture=/Game/Packages/Fortress_Sky/Materials/2Normal}.R
   B=VertexNormalWS()
  )
  n10 = ScalarParameter() {parameterName=DisplaceAmount}
  n11 = CurveAtlasRowParameter(CurveTime=ComponentMask(None=TextureCoordinate() {coordinateIndex=1}) {g=true}) {
   atlas=&TempAthenaCurveAtlas
   curve=&Curve_HT_Tomato_RC
   defaultValue=11
  }
  n12 = TextureCoordinate() {coordinateIndex=1}
  n13 = MaterialFunctionCall(Vector 3=@n1) {materialFunction=&VectorLength}
  n14 = Multiply(A=ScalarParameter() {defaultValue=1, parameterName=SinLength}) {constB=3000}
  n15 = Add(
   A=Multiply(A=ScalarParameter() {parameterName=SinMotion}, B=ScalarParameter() {defaultValue=1, parameterName=SinSpeed})
   B=Multiply(A=ComponentMask(None=@n12) {g=true}, B=Divide(A=@n13.V3 Length, B=@n14) {constB=3000})
  )
  n16 = Normalize(VectorInput=@n1)
  n17 = Tangent!(None=@n16)
  n18 = ScalarParameter() {defaultValue=50, parameterName=SinIntensity}
  n19 = Multiply(A=Multiply(A=Sine(None=@n15), B=Normalize(VectorInput=@n17)), B=@n18)
  n20 = Add(A=@n15, B=ScalarParameter() {parameterName=CosineOffset})
  n21 = CrossProduct(A=@n16, B=@n17)
  n22 = Multiply(A=@n18, B=Multiply(A=Cosine(None=@n20), B=Normalize(VectorInput=@n21)))
  Add(A=Add(A=Add(A=@n2, B=@n6), B=Multiply(A=@n9, B=@n10) {constB=10}), B=Multiply(A=@n11.B, B=Add(A=@n19, B=@n22)))
  n23 = Add(
   A=DynamicParameter() {defaultValue=<rgba 0 1 1 1>, paramNames [Dissolve,Param2,Param3,Param4]}.Dissolve
   B=ScalarParameter() {parameterName=Dissolve}
  )
  n24 = TextureCoordinate() {coordinateIndex=1}
  n25 = WorldPosition()
  n26 = ActorPositionWS()
  n27 = ScalarParameter() {defaultValue=3000, parameterName=Radius}
  n28 = AppendVector(
   A=Multiply(A=ComponentMask(None=@n24) {r=true}, B=ScalarParameter() {defaultValue=1, parameterName=U_Scale})
   B=Multiply(A=SphereMask(A=@n25.XYZ, B=@n26, Radius=@n27) {hardnessPercent=0}, B=ScalarParameter() {defaultValue=1, parameterName=V_Scale})
  )
  n29 = VectorParameter() {parameterName=TextureCtrl, defaultValue=<rgba 1 1 0 0>}
  n30 = Panner(Coordinate=Multiply(A=TextureCoordinate() {coordinateIndex=1}, B=AppendVector(A=@n29.R, B=@n29.R)), Speed=AppendVector(A=@n29.B, B=@n29.B))
  n31 = ScalarParameter() {parameterName=UV_Style}
  n32 = TextureSample(UVs=Add(A=LinearInterpolate(A=@n28, B=@n30, Alpha=@n31), B=Time())) {texture=/Game/Global/Textures/Default/Noise/T_FireNoise_01}
  n33 = Constant()
  LinearInterpolate(
   A=@n23
   B=Add(A=@n23, B=ScalarParameter() {defaultValue=0.5, parameterName=Mask_Add}) {constB=0.5}
   Alpha=Multiply(A=MaterialFunctionCall(In=@n32.R, Contrast=@n33) {materialFunction=&CheapContrast}.Result) {constB=10}
  ) {constB=0.5}
  MaterialFunctionCall(
   Value_Morning=ScalarParameter() {defaultValue=1, group=Time_of_day, parameterName=Morning_Brightness}
   Value_Day=ScalarParameter() {defaultValue=1, group=Time_of_day, parameterName=Day_Brightness}
   Value_Evening=ScalarParameter() {defaultValue=1, group=Time_of_day, parameterName=Evening_Brightness}
   Value_Night=ScalarParameter() {defaultValue=1, group=Time_of_day, parameterName=Night_Brightness}
  ) {materialFunction=&Scalar_Time_of_Day}
  n34 = VectorParameter() {parameterName=UV_Scale, defaultValue=<rgba 1 1 1 1>}
  n35 = TextureSampleParameter2D(UVs=Multiply(A=TextureCoordinate(), B=ComponentMask(None=@n34.RGB) {g=true, r=true})) {parameterName=Tex, texture=/ClashGameplay/Athena/Effects/Particles/Skydiving/NS_Contrail_Clash/T_Skydive_Clash_Tentacles}
  Reroute(None=Multiply(A=@n35.RGB, B=VectorParameter() {parameterName=Color_Tweak, defaultValue=<rgba 2 2 2 1>}.RGB))
  Reroute(None=ScalarParameter() {defaultValue=0.3, parameterName=Metallic})
  Reroute(None=ScalarParameter() {defaultValue=0.4, parameterName=Roughness})
  n36 = SkyAtmosphereLightDirection()
  n37 = TextureSample(UVs=TextureCoordinate() {vTiling=2}) {samplerType=SAMPLERTYPE_Normal, texture=/Game/Global/Textures/Default/Noise/T_SwirlNoise_N}
  n38 = Saturate(None=Power(Base=DotProduct(A=@n36, B=@n37.RGB)) {constExponent=10})
  n39 = Saturate(None=Power(Base=LinearInterpolate(Alpha=@n38) {constA=-10, constB=2}) {constExponent=10})
  n40 = MaterialFunctionCall(
   Color_Morning=VectorParameter() {group=Time_of_day, parameterName=Morning_Color, defaultValue=<rgba 1 1 1 1>}.RGB
   Color_Day=VectorParameter() {group=Time_of_day, parameterName=Day_Color, defaultValue=<rgba 1 1 1 1>}.RGB
   Color_Evening=VectorParameter() {group=Time_of_day, parameterName=Evening_Color, defaultValue=<rgba 1 1 1 1>}.RGB
   Color_Night=VectorParameter() {group=Time_of_day, parameterName=Night_Color, defaultValue=<rgba 1 1 1 1>}.RGB
  ) {materialFunction=&Color_Time_of_Day}
  n41 = Multiply(A=Multiply(A=@n35.RGB) {constB=0.1}, B=VectorParameter() {parameterName=Color_Emissive, defaultValue=<rgba 2 2 2 1>}.RGB)
  Reroute(None=Add(A=Multiply(A=@n39) {constB=50}, B=Multiply(A=@n40.Result, B=@n41)))
  Reroute(None=TextureSampleParameter2D(UVs=TextureCoordinate()) {parameterName=Normal, samplerType=SAMPLERTYPE_Normal, texture=&2Normal}.RGB)
  ScalarParameter() {defaultValue=0.5, parameterName=Specular}
 ]
}
